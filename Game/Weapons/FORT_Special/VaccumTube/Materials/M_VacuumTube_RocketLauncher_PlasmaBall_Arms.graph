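# Translucent, two-sided unlit mesh-particle material, named for the plasma ball arms of the Vacuum Tube rocket launcher.
# A U-axis color gradient tinted by particle color blends with a core color through a mask fading at both U ends, then feeds an emissive-adjustment function.
# That mask also multiplies a panning beam texture and particle alpha.
# Time-scrolled world-space vector noise is scaled by an intensity rising toward the tip.
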Material /Game/Weapons/FORT_Special/VaccumTube/Materials/M_VacuumTube_RocketLauncher_PlasmaBall_Arms
graph {
 properties {
  bUsedWithMeshParticles=true
  bUsedWithNiagaraMeshParticles=true
  blendMode=BLEND_Translucent
  pixelDepthOffsetMode=PDOM_Legacy
  shadingModel=MSM_Unlit
  twoSided=true
  usedShadingModels=Unlit
 }
 nodes [
  n1 = ComponentMask(None=TextureCoordinate()) {r=true}
  n2 = LinearInterpolate(
   A=VectorParameter() {group=Phase 1, parameterName=Color Outline Phase 1_1, defaultValue=<rgba 15 12.0602 0.2851 0>}.RGB
   B=VectorParameter() {group=Phase 1, parameterName=Color Core Phase 1_1, defaultValue=<rgba 5 0.8847 0.0851 0>}.RGB
   Alpha=Clamp(None=Multiply(A=@n1))
  )
  n3 = ParticleColor()
  n4 = ComponentMask(None=TextureCoordinate()) {r=true}
  n5 = OneMinus(None=@n4)
  n6 = Multiply(A=Power(Base=Multiply(A=@n4, B=@n5)) {constExponent=4}, B=ScalarParameter() {defaultValue=50, parameterName=MaskEdgeMult})
  MaterialFunctionCall(Emissive=LinearInterpolate(
    A=VectorParameter() {group=Phase 1, parameterName=Color Core Phase 1_2, defaultValue=<rgba 1 0.115 0.0153 1>}.RGB
    B=Multiply(A=@n2, B=@n3.RGB)
    Alpha=Clamp(None=@n6)
   )) {materialFunction=&MF_FX_EmissiveAdjustments}
  n7 = TextureCoordinate() {uTiling=0.5}
  Multiply(
   A=TextureSample(UVs=Panner(Coordinate=@n7) {speedX=-0.25}) {samplerType=SAMPLERTYPE_LinearColor, texture=/Game/Effects/Fort_Effects/Textures/Tile/T_Raygun_Beam_02}.G
   B=Multiply(A=@n3.A, B=@n6)
  )
  n8 = TextureCoordinate()
  n9 = Power(Base=OneMinus(None=ComponentMask(None=@n8) {r=true}))
  n10 = Subtract(A=WorldPosition().XYZ, B=Multiply(A=ScalarParameter() {defaultValue=5, parameterName=Speed}, B=Time()))
  n11 = ScalarParameter() {defaultValue=50, parameterName=Scale}
  Multiply(
   A=LinearInterpolate(
    A=ScalarParameter() {defaultValue=5, parameterName=Intensity_Tip}
    B=ScalarParameter() {defaultValue=1, parameterName=Intensity}
    Alpha=Clamp(None=@n9)
   )
   B=VectorNoise(World Position=Divide(A=@n10, B=@n11)) {noiseFunction=VNF_VectorALU}
  )
 ]
}
